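# Idle flame effect for the Icicle Cane glider, pairing flipbook sprites with a point light, both attached to its skeletal mesh.
# Eight looping sprites each follow an indexed bone or socket, with curves over spawn group shaping size, offsets and a staggered, overshooting pop-in after a delay.
# The light fades in and scales with time of day.
# VariantPicker switches both emitters' colors, while front-end modifiers shrink and lower the flames.
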NiagaraSystem /BRCosmetics/Gadgets/Assets/VinderTech_GliderChute/IcicleCane/FX/NS_Glider_IcicleCane_Flames_Idle
dependencies {
 usedDataInterfaces [
  &NiagaraDataInterfaceCurve
  &NiagaraDataInterfaceSpriteRendererInfo
  &NiagaraDataInterfaceSkeletalMesh
 ]
 usedDynamicInputs []
 usedModules [
  &self:SystemSpawnScript.SystemScriptSource.SystemScriptGraph.NiagaraNodeAssignment_0.SetVariables_#1
  &self:SystemSpawnScript.SystemScriptSource.SystemScriptGraph.NiagaraNodeAssignment_1.SetVariables_#2
  &SystemState
 ]
}
emitters {
 Flame_White {
  emitterUpdateScript [
   &EmitterState(
    Inactive Response=::NewEnumerator1
    Life Cycle Mode=::NewEnumerator1
    Loop Behavior=::NewEnumerator1
    Loop Delay=1.03
    Loop Duration Mode=::NewEnumerator1
    UseLoopDelay=true
   )
   &SpawnBurst_Instantaneous(Spawn Count=8, Use Spawn Probability=false, UseLoopCountLimit=false)
  ]
  particleSpawnScript [
   &FN_SetSpawnGroup(Number of Spawn Groups=8)
   /SetVariables(Particles.New Normalize Spawn Group:NiagaraFloat=&Divide_Float(
     A=&MakeFloatFromInt(Int=&Multiply_Int(A=$Particles.SpawnGroup, B=2))
     B=16
    )) {isSetParameters=true}
   &InitializeParticle(
    Color Mode=::NewEnumerator0
    Lifetime Max=1.4
    Lifetime Min=0.8
    Lifetime Mode=::NewEnumerator1
    Sprite Size=&Multiply_Vector2D_ByFloat(
     Float=&Multiply_Float(
      A=&RandomRangeFloat(Minimum=40, Maximum=55)
      B=&FloatFromCurve(
       FloatCurve=DataInterface() {propertyValues {Curve {keys [
           {arriveTangent=0.0361, interpMode=RCIM_Cubic, leaveTangent=0.0361, tangentMode=RCTM_User, time=0, value=0.7}
           {arriveTangent=-0.4757, interpMode=RCIM_Cubic, leaveTangent=-3.2871, tangentMode=RCTM_Break, time=0.2, value=1.9022}
           {arriveTangent=-0.7286, interpMode=RCIM_Cubic, leaveTangent=-0.7286, tangentMode=RCTM_User, time=0.55, value=1.1035}
           {arriveTangent=-0.5046, interpMode=RCIM_Cubic, leaveTangent=-0.5046, tangentMode=RCTM_User, time=0.9, value=0.8197}
           <at 1 0>
          ]}}}
       CurveIndex=$Particles.New Normalize Spawn Group
      )
     )
     Vector2D=<1.3034 0.7>
    )
    Sprite Size Mode=::NewEnumerator4
    UsePositionOffset=false
    Write Material Random=true
   )
   &CameraOffset(Camera Offset Amount=&FloatFromCurve(
     FloatCurve=DataInterface() {propertyValues {Curve {keys [
         <at 0 0.7>
         {interpMode=RCIM_Cubic, leaveTangent=-2.3217, tangentMode=RCTM_Break, time=0.2, value=1}
         {interpMode=RCIM_Cubic, tangentMode=RCTM_User, time=1, value=0.3}
        ]}}}
     CurveIndex=$Particles.New Normalize Spawn Group
     Scale Curve=20
    ))
  ]
  particleUpdateScript [
   &ParticleState(
    Kill Particles When Lifetime Has Elapsed=false
    Let Infinitely Lived Particles Die When Emitter Deactivates=true
    Loop Particles Lifetime=true
   )
   &ScaleSpriteSize(
    Scale Sprite Size Mode=::NewEnumerator0
    Uniform Scale Factor=&Multiply_Float(
     A=&FloatFromCurve(
      FloatCurve=DataInterface() {propertyValues {Curve {keys [{tangentMode=RCTM_User, time=0, value=0}, <at 0.2 1.7>, {tangentMode=RCTM_User, time=0.5, value=1}]}}}
      CurveIndex=&Add_Float(
       A=&Add_Float(A=$Emitter.Age, B=-1.03)
       B=&Multiply_Float(A=$Particles.New Normalize Spawn Group, B=-1.7)
      )
     )
     B=&FN_FrontEndFloatModifier(Gameplay=1, Front End=0.65)
    )
   )
   &SkeletalMeshLocation(
    Bone / Socket Index=&Add_Integer(A=&Multiply_Int(A=$Particles.SpawnGroup, B=1), B=0)
    Bone Sampling Mode=::NewEnumerator4
    MaskSpawnGroup=false
    Sampled Position Offset=&AddVectorToPosition(
     Position=&FN_FrontEndVector3Modifier(Gameplay=<0 0 15>, Front End=<0 0 5>)
     Vector=&MakeVector(
      X=0
      Y=0
      Z=&FloatFromCurve(
       FloatCurve=DataInterface() {propertyValues {Curve {keys [<at 0 0.25>,<at 0.05 0.4524>,<at 0.2838 0.6524>,<at 0.45 0.2524>,<at 1 0>]}}}
       CurveIndex=$Particles.New Normalize Spawn Group
       Scale Curve=20
      )
     )
    )
    Skeletal Mesh=$System.Skeletal Mesh Reader
   )
   &SubUVAnimation(
    "Sprite Renderer "=DataInterface() {propertyValues {SpriteRenderer=&self:Flame_White_9.NiagaraSpriteRendererProperties_0}}
    UseEndFrame=true
    UseStartFrame=true
   )
   &ScaleColor(
    Scale RGB=&MakeVectorFromLinearColorRGB(Color=&MergeHelper_ColorSwitch(
      0=<rgba 1.25 1.9674 5 2>
      1=<rgba 0.8 2.4698 20 1>
      2=?
      3=?
      Index=&MakeIntFromFloat(Float=$User.VariantPicker, Conversion Mode=::NewEnumerator1)
      Num=::NewEnumerator0
     ))
    ScaleA=true
    ScaleRGB=true
   )
  ]
  rendererClasses [&NiagaraSpriteRendererProperties]
  renderers [{rendererClass=&NiagaraSpriteRendererProperties}]
 }
 Light {
  emitterUpdateScript [
   &EmitterState(Life Cycle Mode=::NewEnumerator1, Loop Behavior=::NewEnumerator1, Loop Delay=1, Loop Duration Mode=::NewEnumerator1, UseLoopDelay=true)
   &SpawnBurst_Instantaneous(Use Spawn Probability=false, UseLoopCountLimit=false)
  ]
  particleSpawnScript [
   &InitializeParticle(
    Color=&MergeHelper_ColorSwitch(
     0=<rgba 0.5 0.787 2 2>
     1=<rgba 0 2.4349 10 1>
     2=?
     3=?
     Index=&MakeIntFromFloat(Float=$User.VariantPicker)
     Num=::NewEnumerator0
    )
    UsePositionOffset=false
    Write Material Random=true
   )
   &FN_SetSpawnGroup(Number of Spawn Groups=1)
  ]
  particleUpdateScript [
   &ParticleState(Kill Particles When Lifetime Has Elapsed=false, Let Infinitely Lived Particles Die When Emitter Deactivates=true)
   &SkeletalMeshLocation(
    Bone / Socket Index=4
    Bone Sampling Mode=::NewEnumerator4
    MaskSpawnGroup=false
    Sampled Position Offset=<20 0 40>
    Skeletal Mesh=$System.Skeletal Mesh Reader
   )
   &ScaleColor(
    Scale Alpha=&Multiply_Float(A=$System.FadeIn, B=1)
    Scale RGB=&VectorFromFloat(Value=&FNTimeOfDayScalar(Night Scale=0.75, Morning Scale=1.15, Evening Scale=1.15))
    ScaleA=true
    ScaleRGB=true
   )
   &FN_Light_Attributes(Create Exponent Attribute=true, Create Radius Attribute=true, Exponent=2, Radius=75)
  ]
  rendererClasses [&NiagaraLightRendererProperties]
  renderers [{rendererClass=&NiagaraLightRendererProperties, rendererIndex=0}]
 }
}
summary {
 systemName=NS_Glider_IcicleCane_Flames_Idle
 emitters [
  {emitterName=Flame_White, rendererClasses [&NiagaraSpriteRendererProperties]}
  {emitterName=Light, rendererClasses [&NiagaraLightRendererProperties]}
 ]
 userVariables [
  User.bIsFrontEnd {type=NiagaraBool}
  User.bIsFrontEndPreview {type=NiagaraBool}
  User.Grow {type=NiagaraFloat, defaultValue=1, description=""}
  User.VariantPicker {type=NiagaraFloat, defaultValue=0, description=""}
 ]
}
